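; Projectile Blueprint for the ToyFox cannon ability that adds cosmetic feedback when the projectile explodes.
; EventOnExploded calls the parent, then, where cosmetic events can execute, unhides the actor and its Niagara component and deactivates that Niagara effect.
; It then plays the impact camera shake on the instigator's Fortnite player controller only, which the graph comment identifies as the giant.
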(event EventOnExploded
  (params (HitActors :type Array%20of%20Actor%20Object%20References) (HitResults :type Array%20of%20Hit%20Result%20Structures))
  (bind _niagara (Variables|Default|GetNiagara))
  (|CallParentFunction HitActors HitResults)
  (Networking|CanExecuteCosmeticEvents
    (:True
      (Rendering|SetActorHiddenInGame false)
      (Development|SetHiddeninGame _niagara false false)
      (Components|Activation|Deactivate _niagara)
      ;[0 0.02283 0.15 0.5] Camera shake only for the giant.
      (bind _asfort_player_controller_athena (Utilities|Casting|CastToFortPlayerControllerAthena (Game|GetInstigatorController)))
      (Game|Feedback|ClientStartCameraShake :self _asfort_player_controller_athena :Shake (Variables|Default|GetProjectileImpactCameraShakeClass) :Scale 1.0 :PlaySpace "CameraLocal"))))

(fn ConstructionScript)
