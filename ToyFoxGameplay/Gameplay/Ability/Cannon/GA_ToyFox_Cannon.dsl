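; Gameplay ability for the ToyFox giant pawn's cannon: charges, then fires right and left projectiles at the camera trace target.
; Charging applies an effect, cue and camera modifier, slows look rotation by curve, and alerts AI with noise.
; After charging, interval checks require aim and body angles within tolerances before firing.
; Two net syncs order client-predicted firing ahead of the server; recovery restores rotation speed, then ends the ability.
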(event Ability|EventActivateAbility
  (bind _hotfix_cannonrecovertime (Variables|Hotfix|GetHotfix_CannonRecoverTime))
  (bind _returnvalue (ScalableFloat|GetCachedScalableFloatValue _hotfix_cannonrecovertime))
  (bind _hotfix_cannonfireaimdircheckinterval (Variables|Hotfix|GetHotfix_CannonFireAimDirCheckInterval))
  (bind _returnvalue_2 (ScalableFloat|GetCachedScalableFloatValue _hotfix_cannonfireaimdircheckinterval))
  ;[0.001912 0.15 0 0.5] ActivateAbility: Fire the cannon projectiles as long as the pawn is angled correctly.
  (|CallParentFunction)
  (Ability|CommitAbility)
  (CallFunction|CachePlayerController)
  (CallFunction|SetEnableCannonCameraModifier true)
  (Utilities|FlowControl|Sequence
    (:then_0
      ;;[0 0.15 0.1034 0.5] Reduce camera rotation speed over charge up time.
      (if (> (ScalableFloat|GetCachedScalableFloatValue (Variables|Hotfix|GetHotfix_CannonInputRotationSlowdownEnabled)) 0)
        (bind _alpha (Utilities|Tick (ScalableFloat|GetCachedScalableFloatValue (Variables|Hotfix|GetHotfix_CannonChargeTime)))
          (:Tick
            (bind _playercontroller (Variables|Transient|GetPlayerController)
              (:then
                _)
              (:else
                _))
            (Input|SetForcedInputRotationSpeed (Variables|Transient|GetPlayerController) (Math|Float|MapRangeClamped (Math|Curves|GetFloatValue (Variables|Camera|GetForcedInputRotationSpeedCurveEnter) _alpha) 0.0 1.0 (ScalableFloat|GetCachedScalableFloatValue (Variables|Hotfix|GetHotfix_CannonFinalInputRotationSpeed)) (Math|Vector2D|Vector2DLength (Math|Vector2D|MakeVector2D (.pitch (Class|FortPlayerController|GetLastInputRotation _playercontroller)) (.yaw (Class|FortPlayerController|GetLastInputRotation _playercontroller))))))))))
    (:then_1
      ;;[0.03689 0.04106 0.15 0.5] Add noise event for AI to react to
      (bind _outfortplayerpawn (CallFunction|GetAvatarasValidFortPlayerPawn)
        (:ValidFortPlayerPawn
          (AI|Perception|MakeNoiseEventatLocation _outfortplayerpawn (ScalableFloat|GetValueatLevel (Variables|Hotfix|GetHotfix_PerceptionRange)) (Transformation|GetActorLocation _outfortplayerpawn) "WeaponFiring"))))
    (:then_2
      (bind _returnvalue_3 (Ability|ApplyGameplayEffectToOwner (Variables|Settings|GetCannonActiveGameplayEffectClass)))
      (Variables|Transient|SetCannonActiveGameplayEffectHandle _returnvalue_3)
      (Ability|AddGameplayCueWithParamsToOwner (Utilities|Struct|BreakGameplayCueTag (Variables|GameplayCue|GetGameplayCueCannonCharge)) (Ability|GameplayCue|MakeGameplayCueParameters 0.0))
      (Utilities|FlowControl|Delay (ScalableFloat|GetCachedScalableFloatValue (Variables|Hotfix|GetHotfix_CannonChargeTime)))
      (Variables|Transient|SetAimDirValidDuration 0.0)
      (bind (_bvalid _returnvalue_4) (CallFunction|GetProjectileTargetDestination)) ; Runs a trace.
      (if _bvalid
        (bind (_bvalid_2 _desiredaimdirection _currentaimdirection) (CallFunction|GetAimDirectionData _returnvalue_4))
        (if _bvalid_2
          ;; Ensure pawn is angled correctly before firing.
          (bind (_roll _pitch _yaw) (Math|Rotator|BreakRotator (Math|Rotator|Delta(Rotator) (Math|Rotator|MakeRotfromX _desiredaimdirection) (Math|Rotator|MakeRotfromX _currentaimdirection))))
          (Development|PrintString (Utilities|String|Append "ToyFoxCannonAimAngleDiff: Pitch: " (Utilities|String|ToString(Float) (Math|Float|Absolute(Float) _pitch)) " : Yaw : " (Utilities|String|ToString(Float) (Math|Float|Absolute(Float) _yaw))) true false "(R=0.407053,G=0.603628,B=1.000000,A=1.000000)" 2.0 "ToyFoxCannonAimAngle")
          (if (and (< (Math|Float|Absolute(Float) _pitch) (ScalableFloat|GetCachedScalableFloatValue (Variables|Hotfix|GetHotfix_CannonFireAimDirMaxPitchDelta))) (< (Math|Float|Absolute(Float) _yaw) (ScalableFloat|GetCachedScalableFloatValue (Variables|Hotfix|GetHotfix_CannonFireAimDirMaxYawDelta))))
            (bind _returnvalue_5 (Transformation|GetBoneTransform (Variables|/Script/Engine.Character|GetMesh (CallFunction|GetAvatarasFortPlayerPawn)) "pelvis" "RTS_World"))
            (bind (_returnvalue_location _returnvalue_rotation) (Math|Transform|BreakTransform _returnvalue_5))
            (Development|PrintString (Utilities|String|Append "ToyFoxCannonBodyAngleDiff: Yaw: " (Utilities|String|ToString(Float) (Math|Float|Absolute(Float) (.yaw (Math|Rotator|Delta(Rotator) (Math|Rotator|MakeRotfromX (Math|Vector|GetUnitDirection(Vector) _returnvalue_location _returnvalue_4)) (Math|Rotator|MakeRotfromX (Math|Vector|GetRightVector _returnvalue_rotation))))))) true false "(R=0.407053,G=0.603628,B=1.000000,A=1.000000)" 2.0 "ToyFoxCannonBodyAngleDiff")
            (if (< (Math|Float|Absolute(Float) (.yaw (Math|Rotator|Delta(Rotator) (Math|Rotator|MakeRotfromX (Math|Vector|GetUnitDirection(Vector) _returnvalue_location _returnvalue_4)) (Math|Rotator|MakeRotfromX (Math|Vector|GetRightVector _returnvalue_rotation))))) (ScalableFloat|GetCachedScalableFloatValue (Variables|Hotfix|GetHotfix_CannonFireBodyAimDirMaxYawDelta)))
              (bind _output_get (Variables|Transient|SetAimDirValidDuration (+ (Variables|Transient|GetAimDirValidDuration) _returnvalue_2)))
              (if (not (> _output_get (ScalableFloat|GetCachedScalableFloatValue (Variables|Hotfix|GetHotfix_CannonFireAimValidDurationThreshold))))
                (Utilities|FlowControl|Delay _returnvalue_2))
              (bind _asynctaskproxy (Ability|Tasks|WaitNetSync "BothWait")
                (:then
                  (Utilities|Array|Add (Variables|Transient|GetAsyncTask_WaitNetSyncs) _asynctaskproxy))
                (:OnSync
                  (bind _asynctaskproxy_2 (Ability|Tasks|WaitNetSync "OnlyServerWait")
                    (:then
                      (Utilities|Array|Add (Variables|Transient|GetAsyncTask_WaitNetSyncs) _asynctaskproxy_2))
                    (:OnSync
                      (Utilities|FlowControl|Sequence
                        (:then_0
                          (Ability|RemoveGameplayCueFromOwner (Utilities|Struct|BreakGameplayCueTag (Variables|GameplayCue|GetGameplayCueCannonCharge)))
                          (Utilities|FlowControl|Sequence
                            (:then_0
                              ;;[0.01843 0.0118 0.15 0.5] Increase the reduced camera rotation speed back to default over recover time.
                              (if (> (ScalableFloat|GetCachedScalableFloatValue (Variables|Hotfix|GetHotfix_CannonInputRotationSlowdownEnabled)) 0)
                                (bind _alpha_2 (Utilities|Tick _returnvalue)
                                  (:Tick
                                    (bind _playercontroller_2 (Variables|Transient|GetPlayerController)
                                      (:then
                                        _)
                                      (:else
                                        _))
                                    (bind _playercontroller_3 (Variables|Transient|GetPlayerController)
                                      (:then
                                        (Input|SetForcedInputRotationSpeed _playercontroller_3 (Math|Float|MapRangeClamped (Math|Curves|GetFloatValue (Variables|Camera|GetForcedInputRotationSpeedCurveExit) _alpha_2) 0.0 1.0 (ScalableFloat|GetCachedScalableFloatValue (Variables|Hotfix|GetHotfix_CannonFinalInputRotationSpeed)) (Math|Vector2D|Vector2DLength (Math|Vector2D|MakeVector2D (.pitch (Class|FortPlayerController|GetLastInputRotation _playercontroller_2)) (.yaw (Class|FortPlayerController|GetLastInputRotation _playercontroller_2)))))))
                                      (:else))))))
                            (:then_1
                              (Utilities|FlowControl|Delay _returnvalue)
                              (Ability|EndAbility))))
                        (:then_1
                          (Utilities|FlowControl|Delay 0.033)
                          (Utilities|FlowControl|Sequence
                            (:then_0
                              ;; Cannon is firing!
                              (Ability|ApplyGameplayEffectToOwner (Variables|Settings|GetFiringGameplayEffectClass)) ; Applies for a brief duration.
                              ;; Fire gameplay cue.
                              (Ability|ExecuteGameplayCueWithParamsOnOwner (Utilities|Struct|BreakGameplayCueTag (Variables|GameplayCue|GetGameplayCueCannonFire3P)) (Ability|GameplayCue|MakeGameplayCueParameters 0.0))
                              (GameplayCue|ExecuteGameplayCue%28Burst%29Local (Ability|GetAvatarActorfromActorInfo) (Utilities|Struct|BreakGameplayCueTag (Variables|GameplayCue|GetGameplayCueCannonFire1P)) (Variables|Transient|GetDefaultGameplayCueParams) false) ; Running 1P separately for responsiveness \nsince there's no prediction key at this point.
                            )
                            (:then_1
                              ;; Spawn cannon projectiles.
                              (Utilities|FlowControl|Delay 0.033)
                              (CallFunction|SpawnCannonProjectile true (CallFunction|AdjustProjectileTargetLocation true _returnvalue_4 _currentaimdirection))
                              (Utilities|FlowControl|Delay 0.033)
                              (CallFunction|SpawnCannonProjectile false (CallFunction|AdjustProjectileTargetLocation false _returnvalue_4 _currentaimdirection)))))))) ; Server waits for the client to fire projectile first since it is client predicted.
                )) ; Both server and client should have aimed within the threshold.
            )
            (Variables|Transient|SetAimDirValidDuration 0.0)
            (Utilities|FlowControl|Delay (ScalableFloat|GetCachedScalableFloatValue (Variables|Hotfix|GetHotfix_CannonFireAimDirCheckInterval)))))
        (Ability|CancelAbility)))))

(event Ability|EventOnEndAbility
  (params (bWasCancelled :type Boolean))
  ;[0.125 0.07241 0 0.5] OnEndAbility: Clean up.
  (Utilities|FlowControl|Sequence
    (:then_0
      (Ability|RemoveGameplayEffectFromOwnerWithHandle (Variables|Transient|GetCannonActiveGameplayEffectHandle))
      (GameplayTasks|EndTask (Variables|Transient|GetAsyncTask_WaitNetSyncs))
      (CallFunction|SetEnableCannonCameraModifier false)
      (Ability|RemoveGameplayCueFromOwner (Utilities|Struct|BreakGameplayCueTag (Variables|GameplayCue|GetGameplayCueCannonCharge))) ; In case this ability is canceled abruptly
      (bind _playercontroller (Variables|Transient|GetPlayerController)
        (:then
          (Input|ResetForcedInputRotationSpeed _playercontroller))
        (:else)))
    (:then_1
      (|CallParentFunction bWasCancelled))))

(fn AdjustProjectileTargetLocation
  (params (bRightCannon :type Boolean) (TargetLocation :type Vector) (AimDirection :type Vector))
  (return (Utilities|Operators|Add TargetLocation (* (* (Math|Vector|Normalize (Math|Vector|GetRightVector (Math|Rotator|MakeRotfromX (Math|Conversions|ToVector(Vector2D) (Math|Conversions|ToVector2D(Vector) AimDirection))))) (ScalableFloat|GetCachedScalableFloatValue (Variables|Hotfix|GetHotfix_CannonProjectileTargetLocationOffset))) (Math|Float|SelectFloat 1.0 -1.0 bRightCannon)) (Utilities|Operators|Multiply (Math|Conversions|ToVector2D(Vector) (Transformation|GetVelocity (CallFunction|GetAvatarasFortPlayerPawn))) (Utilities|Time|GetWorldDeltaSeconds) (ScalableFloat|GetCachedScalableFloatValue (Variables|Hotfix|GetHotfix_PrjSpawnLocationForwardXYOffsetVelocityBasedMult))))))

(fn CachePlayerController
  (bind _outfortplayerpawn (CallFunction|GetAvatarasValidFortPlayerPawn)
    (:ValidFortPlayerPawn
      (Variables|Transient|SetPlayerController (Fortnite|GetFortPlayerControllerfromActor _outfortplayerpawn)))))

(fn GetAimDirectionData
  (params (TargetDestination :type Vector))
  (Utilities|FlowControl|Sequence
    (:then_0
      (bind _outfortplayerpawn (CallFunction|GetAvatarasValidFortPlayerPawn)
        (:ValidFortPlayerPawn
          (bind _astoy_fox_giant_pawn_interface (Utilities|Casting|CastToToyFoxGiantPawnInterface _outfortplayerpawn))
          (bind _mesh (Class|ToyFoxGiantPawnInterface|GetSkeletalMeshBudgeted _astoy_fox_giant_pawn_interface))
          (return true (Math|Vector|Normalize (- TargetDestination (.location (Transformation|GetSocketTransform _mesh (Variables|Settings|GetAimSourceBoneName) "RTS_World")))) (Math|Vector|GetRightVector (.rotation (Transformation|GetSocketTransform _mesh (Variables|Settings|GetAimSourceBoneName) "RTS_World")))))))
    (:then_1
      (return false "0, 0, 0" "0, 0, 0"))))

(fn GetProjectileTargetDestination
  (Utilities|FlowControl|Sequence
    (:then_0
      (bind _outfortplayerpawn (CallFunction|GetAvatarasValidFortPlayerPawn)
        (:ValidFortPlayerPawn
          (bind (_outhit _returnvalue) (Collision|LineTraceByChannel (.location (Targeting|GetTargetingTransform _outfortplayerpawn "Camera" self)) (+ (.location (Targeting|GetTargetingTransform _outfortplayerpawn "Camera" self)) (* (Math|Vector|GetForwardVector (.rotation (Targeting|GetTargetingTransform _outfortplayerpawn "Camera" self))) 1e+05)) "TraceTypeQuery5" true (Utilities|Array|MakeArray (CallFunction|GetAvatarasFortPlayerPawn)) (select (Variables|Debug|GetbDebug) "ForDuration" "None")))
          (return true (Math|Vector|SelectVector (.ImpactPoint (Collision|BreakHitResult _outhit)) (.TraceEnd (Collision|BreakHitResult _outhit)) _returnvalue)))))
    (:then_1
      (return false "0, 0, 0"))))

(fn OverrideFailedReason
  (params (FailedReason :type Gameplay%20Tag%20Container%20Structure%20%28by%20ref%29))
  ;[0 0.15 0.08673 0.5] Do not display fail reason on the HUD.
  (return 0))

(fn SetEnableCannonCameraModifier
  (params (bEnable :type Boolean))
  (bind _playercontroller (Variables|Transient|GetPlayerController)
    (:then
      (if bEnable
        (Variables|Transient|GetCameraModifier
          (:then)
          (:else
            (bind _returnvalue (Fortnite|Camera|AddCameraModifierforOwningPlayerController "/ToyFoxGameplay/Gameplay/Ability/Cannon/Camera/CameraModifierFrustrum_ToyFox_Cannon.CameraModifierFrustrum_ToyFox_Cannon_C" _playercontroller))
            (Variables|Transient|SetCameraModifier _returnvalue)))
        (else
          (bind _playercameramanager (Variables|/Script/Engine.PlayerController|GetPlayerCameraManager _playercontroller)
            (:then
              (Game|Player|RemoveCameraModifier _playercameramanager (Variables|Transient|GetCameraModifier))
              (Variables|Transient|SetCameraModifier 0))
            (:else)))))
    (:else)))

(fn SpawnCannonProjectile
  (params (bRightCannon :type Boolean) (ProjectileTargetDestination :type Vector))
  (bind _astoy_fox_giant_pawn_interface (Utilities|Casting|CastToToyFoxGiantPawnInterface (CallFunction|GetAvatarasFortPlayerPawn)))
  (bind (_left_location _right_location) (Class|ToyFoxGiantPawnInterface|GetCannonMuzzleLocations _astoy_fox_giant_pawn_interface))
  (Projectile|SpawnClientPredictedProjectile "/ToyFoxGameplay/Gameplay/Ability/Cannon/B_Prj_ToyFoxCannon.B_Prj_ToyFoxCannon_C" (CallFunction|GetAvatarasFortPlayerPawn) (Utilities|Struct|MakeSpawnProjectileParams (select bRightCannon _right_location _left_location) (Math|Conversions|RotationFromXVector (Math|Vector|Normalize (- ProjectileTargetDestination (select bRightCannon _right_location _left_location)))) (Math|Conversions|RotationFromXVector (Math|Vector|Normalize (- ProjectileTargetDestination (select bRightCannon _right_location _left_location)))) "(TargetSelection=(List=,bStopAtFirstSuccess=True,bKeepCheckingListOnIndestructibleHit=False,AbilityTargetBlockedGameplayTags=(GameplayTags=),bUseWeaponRanges=True,bUseMaxYawAngleToTarget=False,MaxYawAngleToTarget=0.000000),TargetGameplayEffectSpecs=,OwnerGameplayEffectSpecs=,ActivationCues=(GameplayTags=),ImpactCues=(GameplayTags=),ImpactNoiseRange=0.000000,FlyByNoiseRange=0.000000,bOverrideChargeMagnitude=False,ChargeMagnitudeOverrideValue=0.000000)" "(TargetSelection=(List=,bStopAtFirstSuccess=True,bKeepCheckingListOnIndestructibleHit=False,AbilityTargetBlockedGameplayTags=(GameplayTags=),bUseWeaponRanges=True,bUseMaxYawAngleToTarget=False,MaxYawAngleToTarget=0.000000),TargetGameplayEffectSpecs=,OwnerGameplayEffectSpecs=,ActivationCues=(GameplayTags=),ImpactCues=(GameplayTags=),ImpactNoiseRange=0.000000,FlyByNoiseRange=0.000000,bOverrideChargeMagnitude=False,ChargeMagnitudeOverrideValue=0.000000)" false 0 (ScalableFloat|GetCachedScalableFloatValue (Variables|Hotfix|GetHotfix_CannonProjectileSpeed)) 0.0 0.0 0.0 0 (Utilities|Struct|MakeFortProjectilePredictionKey)))
  (return))
