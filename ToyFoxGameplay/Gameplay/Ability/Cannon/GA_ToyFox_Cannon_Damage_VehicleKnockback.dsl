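; Gameplay ability for cannon damage vehicle knockback in ToyFox, overriding where the knockback originates.
; GetKnockbackOrigin takes the source object from the granting effect's context and returns its location when it casts to an Actor.
; A Sequence node provides the fallback: if the cast fails, the parent implementation runs with OriginActor.
; The activate and end events are empty stubs, so the child adds no other logic.
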(event Ability|EventActivateAbility
  (|CallParentFunction) ; This node is disabled and will not be called.\nDrag off pins to build functionality.
)

(event Ability|EventOnEndAbility
  (params (bWasCancelled :type Boolean)))

(fn GetKnockbackOrigin
  (params (OriginActor :type Actor%20Object%20Reference))
  (Utilities|FlowControl|Sequence
    (:then_0
      (bind _asactor (Utilities|Casting|CastToActor (Ability|EffectContext|GetSourceObject (Ability|GetGrantedbyEffectContext))))
      (return (Transformation|GetActorLocation _asactor)))
    (:then_1
      (return (|CallParentFunction OriginActor)))))
