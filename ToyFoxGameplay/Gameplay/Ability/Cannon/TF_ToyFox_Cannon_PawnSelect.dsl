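; Targeting filter for the ToyFox cannon ability that keeps pawn-like hit actors and removes all other targets.
; ShouldRemoveTarget pulls the hit actor from the targeting result and rejects it when invalid.
; It keeps actors that are a FortPawn or FortPhysicsPawn, checked by soft class path.
; Other actors survive only with a matching gameplay tag: ProtoSwarm NPC, ToyFox boss, or the generic ConsiderAsPawn override.
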(fn ShouldRemoveTarget
  (params (TargetingHandle :type Targeting%20Request%20Handle%20Structure%20%28by%20ref%29) (TargetData :type Targeting%20Default%20Result%20Data%20Structure%20%28by%20ref%29))
  (bind _hitresult (Utilities|Struct|BreakTargetingDefaultResultData TargetData))
  (bind (_bblockinghit _binitialoverlap _time _distance _location _impactpoint _normal _impactnormal _physmat _hitactor) (Collision|BreakHitResult _hitresult))
  (Utilities|IsValid _hitactor
    (:Is%20Valid
      (Utilities|IsA(soft) _hitactor "/Script/FortniteGame.FortPawn"
        (:True
          _)
        (:False
          (Utilities|IsA(soft) _hitactor "/Script/FortniteGame.FortPhysicsPawn"
            (:True
              _)
            (:False
              (bind _returnvalue (GameplayTags|HasAnyMatchingGameplayTags(Message) _hitactor (GameplayTags|MakeLiteralGameplayTagContainer "(GameplayTags=((TagName=\"NPC.ProtoSwarm\"),(TagName=\"TidalCrane.ToyFox.Boss\"),(TagName=\"Athena.Ability.TargetFilter.ConsiderAsPawn\")))")))
              (if (not _returnvalue)
                (return true))))))
      (return false))
    (:Is%20Not%20Valid
      (return true))))
